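; Enhanced Input modifier for UI pointer zoom that combines a zoom delta with screen-space coordinates.
; ModifyRaw calls the parent implementation on the incoming value and keeps that result's Y and Z components.
; It returns an Axis3D value with the raw input's X as the zoom delta alongside those Y and Z coordinates.
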(fn ModifyRaw
  (params (PlayerInput :type Enhanced%20Player%20Input%20Object%20Reference) (CurrentValue :type Input%20Action%20Value%20Structure) (DeltaTime :type Float%20%28single-precision%29))
  (bind (_x _y _z) (Input|BreakInputActionValue (|CallParentFunction PlayerInput CurrentValue DeltaTime)))
  ; We want the delta of the zoom, and to combine it with the screen space coordinates passed in.
  (return (Input|MakeInputActionValueOfType (Input|BreakInputActionValue CurrentValue) _y _z "Axis3D")))
